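# Niagara effect for the Evil Cowboy back bling's stage 3 transition, likely a short burst around a cube-shaped prop.
# Lightning meshes spawn at eight offset cube corners along edge axes, layered with wisp meshes, runes, light rays, upward bursts, smoke, swirls and a glow with light and flare.
# System.VariantIndex switches color palettes, mostly magenta versus cyan.
# Front-end modifiers dim rays and bursts in preview.
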NiagaraSystem /BRCosmetics/Accessories/FORT_Backpacks/M_MED_Evil_Cowboy_Backpack/FX/NS_BackPack_EvilCowboy_Stage3_Transition
dependencies {
 usedDataInterfaces [
  &NiagaraDataInterfaceCurve
  &NiagaraDataInterfaceArrayFloat3
  &NiagaraDataInterfaceVectorCurve
  &NiagaraDataInterfaceSpriteRendererInfo
  &NiagaraDataInterfaceVector2DCurve
  &NiagaraDataInterfaceSkeletalMesh
 ]
 usedDynamicInputs [&MakeQuaternion, &MakeIntFromFloat]
 usedModules [
  &self:SystemSpawnScript.SystemScriptSource.SystemScriptGraph.NiagaraNodeAssignment_0.SetVariables_#1
  &SystemState
  &self:SystemSpawnScript.SystemScriptSource.SystemScriptGraph.NiagaraNodeAssignment_1.SetVariables_#2
 ]
}
emitters {
 Cube_WispAway {
  emitterUpdateScript [
   &EmitterState(Life Cycle Mode=::NewEnumerator1, Loop Behavior=::NewEnumerator1, Loop Duration=0.2, UseLoopDelay=true)
   &SpawnBurst_Instantaneous(Spawn Count=2, Use Spawn Probability=false, UseLoopCountLimit=false)
  ]
  particleSpawnScript [
   &InitializeParticle(Mesh Scale Mode=::NewEnumerator4, UsePositionOffset=false, Write Material Random=true)
   &DynamicMaterialParameters(
    Index 0 Param 2=20
    Module.Param0WriteEnabled=false
    Module.Param1WriteEnabled=true
    Module.Param2WriteEnabled=false
    Module.Param3WriteEnabled=false
   ) {module=DynamicMaterialParameters001}
   &Color(Color=&MergeHelper_ColorSwitch(
     0=&RandomRangeLinearColor(Minimum=<rgba 1 25 1 100>, Maximum=<rgba 4 25 1 100>)
     1=&RandomRangeLinearColor(Minimum=<rgba 0.3208 19.4582 25 100>, Maximum=<rgba 0 22.6072 25 100>)
     2=?
     3=?
     Index=$System.VariantIndex
     Num=::NewEnumerator0
    ))
   /CascadeToNiagaraConverter/NiagaraScripts/ModuleScripts/AddVectorToPosition(Vector:Vector3f=&TransformVector(Vector=<0 -10 -10>, Destination Space=::Simulation))
   /SetVariables(Particles.VisibilityTag=$System.VariantIndex) {isSetParameters=true}
  ]
  particleUpdateScript [
   &ParticleState()
   &ScaleMeshSize(Scale Factor=&VectorFromCurve(VectorCurve=DataInterface() {propertyValues {XCurve {keys [<at 0 1>,<at 1 1>]}, YCurve {keys [<at 0 1>,<at 1 1>]}, ZCurve {keys [<at 0 1>,<at 1 4>]}}}))
   &CameraOffset(Camera Offset Amount=6)
   &DynamicMaterialParameters(
    Index 0 Param 1=&FloatFromCurve(FloatCurve=DataInterface() {propertyValues {Curve {keys [<at 0 5>,<at 1 2>]}}})
    Index 0 Param 3=&FloatFromCurve(FloatCurve=DataInterface() {propertyValues {Curve {keys [
         {arriveTangent=-3.5699, interpMode=RCIM_Cubic, leaveTangent=-3.5699, tangentMode=RCTM_User, time=0, value=0}
         {arriveTangent=24.5202, interpMode=RCIM_Cubic, leaveTangent=24.5202, time=0.2386, value=2.2658}
         {arriveTangent=10.1581, interpMode=RCIM_Cubic, leaveTangent=10.1581, time=0.5307, value=13.0137}
         <at 1 10 RCIM_Cubic>
        ]}}})
    Index 0 Param 4=&FloatFromCurve(FloatCurve=DataInterface() {propertyValues {Curve {keys [
         {arriveTangent=-0.0169, interpMode=RCIM_Cubic, leaveTangent=-0.0169, tangentMode=RCTM_User, time=0.0123, value=1}
         {arriveTangent=-1.9945, interpMode=RCIM_Cubic, leaveTangent=-1.9945, time=0.4921, value=0.1511}
         {arriveTangent=-2.2661, interpMode=RCIM_Cubic, leaveTangent=-2.2661, time=0.8937, value=-0.7578}
         <at 1 -1 RCIM_Cubic>
        ]}}})
    Module.Param0WriteEnabled=true
    Module.Param1WriteEnabled=false
    Module.Param2WriteEnabled=true
    Module.Param3WriteEnabled=true
   )
   &ScaleColor(
    Scale Alpha=&FloatFromCurve(FloatCurve=DataInterface() {propertyValues {Curve {keys [<at 0 1 RCIM_Cubic>, {arriveTangent=-1, interpMode=RCIM_Cubic, leaveTangent=-1, time=0.5734, value=0.5233}, <at 1 0 RCIM_Cubic>]}}})
    Scale RGB=&VectorFromCurve(VectorCurve=DataInterface() {propertyValues {
       XCurve {keys [
         {interpMode=RCIM_Cubic, tangentMode=RCTM_User, time=-0.0019, value=35}
         {arriveTangent=-33.7364, interpMode=RCIM_Cubic, leaveTangent=-33.7364, time=0.1894, value=1.0285}
         <at 1 1.2>
        ]}
       YCurve {keys [
         {arriveTangent=-255.9352, interpMode=RCIM_Cubic, leaveTangent=-255.9352, tangentMode=RCTM_User, time=-0.0019, value=34.398}
         {arriveTangent=-34.0837, interpMode=RCIM_Cubic, leaveTangent=-34.0837, time=0.1894, value=2.6772}
         <at 1 0.25>
        ]}
       ZCurve {keys [
         {interpMode=RCIM_Cubic, tangentMode=RCTM_User, time=-0.0019, value=35}
         {arriveTangent=-34.9341, interpMode=RCIM_Cubic, leaveTangent=-34.9341, time=0.1894, value=0.8576}
         <at 1 0>
        ]}
      }})
    ScaleA=true
    ScaleRGB=true
   )
  ]
  rendererClasses [&NiagaraMeshRendererProperties]
  renderers [
   {rendererClass=&NiagaraMeshRendererProperties, rendererIndex=0}
   {rendererClass=&NiagaraMeshRendererProperties, rendererIndex=1}
  ]
 }
 InkySmokeRadial02 {
  emitterUpdateScript [
   &EmitterState(Life Cycle Mode=::NewEnumerator1, Loop Behavior=::NewEnumerator1, Loop Duration=0.1, UseLoopDelay=true)
   &SpawnRate(SpawnRate=50, Use Spawn Probability=false)
  ]
  particleSpawnScript [
   /SetVariables(Particles.Acceleration:Vector3f=&RandomRangeVector(Minimum=<0 -25 2>, Maximum=<0 -35 2>, Evaluation Type=::NewEnumerator1)) {isSetParameters=true}
   &InitializeParticle(
    Lifetime=&RandomRangeFloat(Minimum=0.4, Maximum=0.8, Evaluation Type=::NewEnumerator1)
    Sprite Rotation Angle=&RandomRangeFloat(Minimum=0.2, Maximum=1, Evaluation Type=::NewEnumerator1)
    Sprite Rotation Mode=::NewEnumerator1
    Sprite Size=&Vector2DFromFloat(Value=&RandomRangeFloat(Minimum=10, Maximum=15, Evaluation Type=::NewEnumerator1))
    Sprite Size Mode=::NewEnumerator4
    UsePositionOffset=false
    Write Material Random=true
   )
   &Color(Color=&MergeHelper_ColorSwitch(
     0=&RandomRangeLinearColor(Minimum=<rgba 1 0.03 0.45 0.25>, Maximum=<rgba 1 0.1 0.88 1>)
     1=&RandomRangeLinearColor(Minimum=<rgba 0.0038 0.709 1 0.25>, Maximum=<rgba 0.0177 0.7586 1 1>)
     2=?
     3=?
     Index=$System.VariantIndex
     Num=::NewEnumerator0
    ))
   &AddVelocity(
    Rotation Coordinate Space=::World
    Velocity=&RandomRangeVector(Minimum=<0 0 5>, Maximum=<0 0 25>, Evaluation Type=::NewEnumerator1)
   )
   &DynamicMaterialParameters(
    Index 0 Param 1=&RandomRangeFloat(Minimum=0, Maximum=4, Evaluation Type=::NewEnumerator1)
    Index 0 Param 2=&RandomRangeFloat(Minimum=1, Maximum=4, Evaluation Type=::NewEnumerator1)
    Module.Param0WriteEnabled=true
    Module.Param1WriteEnabled=true
    Module.Param2WriteEnabled=false
    Module.Param3WriteEnabled=false
   )
   &ShapeLocation(MaskSpawnGroup=false, Override Seed=false, Sphere Radius=5, UseOwnerScale=true)
   /CascadeToNiagaraConverter/NiagaraScripts/ModuleScripts/FixShapeVector()
   &AddVelocity(Velocity=&Multiply_VectorByFloat(Vector=$Output.FixShapeVector.ShapeVector, Float=2)) {module=AddVelocity001}
  ]
  particleUpdateScript [
   &ParticleState()
   &ScaleSpriteSize(
    Scale Sprite Size Mode=::NewEnumerator0
    Uniform Scale Factor=&FloatFromCurve(FloatCurve=DataInterface() {propertyValues {Curve {keys [
         {arriveTangent=16.9657, interpMode=RCIM_Cubic, leaveTangent=16.9657, tangentMode=RCTM_User, time=0, value=1}
         {arriveTangent=4.2114, interpMode=RCIM_Cubic, leaveTangent=4.2114, time=0.2637, value=3.9734}
         <at 1.0038 5.2275 RCIM_Cubic>
        ]}}})
   )
   &ScaleColor(
    Scale Alpha=&FloatFromCurve(FloatCurve=DataInterface() {propertyValues {Curve {keys [
         <at 0 0 RCIM_Cubic>
         {arriveTangent=3.8807, interpMode=RCIM_Cubic, leaveTangent=3.8807, tangentMode=RCTM_User, time=0.1167, value=0.9001}
         {arriveTangent=-3.7324, interpMode=RCIM_Cubic, leaveTangent=-3.7324, tangentMode=RCTM_User, time=0.7429, value=0.7077}
         <at 1 0 RCIM_Cubic>
        ]}}})
    Scale RGB=&VectorFromCurve(VectorCurve=DataInterface() {propertyValues {
       XCurve {keys [<at 0 5 RCIM_Cubic>, {arriveTangent=-14.2857, interpMode=RCIM_Cubic, leaveTangent=-14.2857, time=0.1, value=0.8785}, <at 0.35 0 RCIM_Cubic>]}
       YCurve {keys [<at 0 5 RCIM_Cubic>, {arriveTangent=-14.2857, interpMode=RCIM_Cubic, leaveTangent=-14.2857, time=0.1, value=0.1156}, <at 0.35 0 RCIM_Cubic>]}
       ZCurve {keys [<at 0 5 RCIM_Cubic>, {arriveTangent=-14.2857, interpMode=RCIM_Cubic, leaveTangent=-14.2857, time=0.1, value=2.0242}, <at 0.35 0 RCIM_Cubic>]}
      }})
    ScaleA=true
    ScaleRGB=true
   )
   &AccelerationForce(Acceleration=$Particles.Acceleration)
   &Drag(Drag=2, Use Linear Drag=true, Use Rotational Drag=true)
   &SolveForcesAndVelocity(Clamp Velocity=false, Limit Acceleration=false)
   &DynamicMaterialParameters(
    Index 0 Param 3=120
    Index 0 Param 4=0.5
    Module.Param0WriteEnabled=false
    Module.Param1WriteEnabled=false
    Module.Param2WriteEnabled=true
    Module.Param3WriteEnabled=true
   ) {module=DynamicMaterialParameters001}
   &CameraOffset(Camera Offset Amount=&FloatFromCurve(FloatCurve=DataInterface() {propertyValues {Curve {keys [<at 0 50>,<at 0.5 -5>]}}}))
   &SubUVAnimation(
    Frames Per Second=32
    Playback Mode:ENiagara_FPSPlayrate=::NewEnumerator1
    "Sprite Renderer "=DataInterface() {propertyValues {SpriteRenderer=&self:InkySmokeRadial02_0.NiagaraSpriteRendererProperties_9}}
    Start Frame Offset=32
    SubUV Animation Mode:ENiagara_SubUVLookupModeV2=::NewEnumerator3
    UseEndFrame=false
    UsePlayRateScale=false
    UseStartFrame=false
   )
  ]
  rendererClasses [&NiagaraSpriteRendererProperties]
  renderers [{rendererClass=&NiagaraSpriteRendererProperties}]
 }
 LightRays {
  emitterUpdateScript [
   &EmitterState(Life Cycle Mode=::NewEnumerator1, Loop Behavior=::NewEnumerator1, Loop Duration=0.125, UseLoopDelay=true)
   &SpawnRate(SpawnRate=150, Use Spawn Probability=false)
  ]
  particleSpawnScript [
   &InitializeParticle(
    Lifetime=&RandomRangeFloat(Minimum=0.1, Maximum=0.2, Evaluation Type=::NewEnumerator1)
    Sprite Rotation Angle=&RandomRangeFloat(Minimum=-1, Maximum=1, Evaluation Type=::NewEnumerator1)
    Sprite Rotation Mode=::NewEnumerator1
    Sprite Size=&RandomRangeVector2D(Minimum=<15 35>, Maximum=<25 40>, Evaluation Type=::NewEnumerator1)
    Sprite Size Mode=::NewEnumerator4
    UsePositionOffset=false
    Write Material Random=true
   )
   &Color(Color=&MergeHelper_ColorSwitch(
     0=<rgba 15 1.85 11.81 0.04>
     1=&MakeCustomLinearColorFromBool(Bool=$User.bIsFrontEndPreview, True Color=<rgba 0.0882 3.3436 5 0.04>, False Color=<rgba 0.2645 10.0308 15 0.04>)
     2=?
     3=?
     Index=$System.VariantIndex
     Num=::NewEnumerator0
    ))
   &AddVectorToPosition(Vector:Vector3f=&TransformVector(Vector=<0 -10 0>, Destination Space=::Simulation))
   &ShapeLocation(MaskSpawnGroup=false, Override Seed=false, Sphere Radius=1, Sphere Surface Distribution=1, UseOwnerScale=true)
  ]
  particleUpdateScript [
   &ParticleState()
   &ScaleSpriteSize(
    Scale Factor=&Vector2DFromCurve(
     Vector2Curve=DataInterface() {propertyValues {
       XCurve {keys [
         {arriveTangent=4.9957, interpMode=RCIM_Cubic, leaveTangent=4.9957, tangentMode=RCTM_User, time=-0.0026, value=1}
         {arriveTangent=0.9974, interpMode=RCIM_Cubic, leaveTangent=0.9974, time=0.3904, value=1.8714}
         <at 1 2 RCIM_Cubic>
        ]}
       YCurve {keys [
         {arriveTangent=0.3115, interpMode=RCIM_Cubic, leaveTangent=0.3115, tangentMode=RCTM_User, time=-0.0026, value=0.8481}
         {arriveTangent=0.1515, interpMode=RCIM_Cubic, leaveTangent=0.1515, time=0.3904, value=0.9697}
         <at 1 1 RCIM_Cubic>
        ]}
      }}
     ScaleCurve=<1 1>
    )
    Scale Sprite Size Mode=::NewEnumerator2
   )
   &DynamicMaterialParameters(
    Index 0 Param 1=1
    Index 0 Param 2=1
    Index 0 Param 3=1
    Index 0 Param 4=1
    Module.Param0WriteEnabled=true
    Module.Param1WriteEnabled=true
    Module.Param2WriteEnabled=true
    Module.Param3WriteEnabled=true
   )
   &SubUVAnimation(
    End Frame Range Override=2
    "Sprite Renderer "=DataInterface() {propertyValues {SpriteRenderer=&self:LightRays_0.NiagaraSpriteRendererProperties_4}}
    UseEndFrame=false
    UseStartFrame=false
   )
   &ScaleColor(
    Scale Alpha=&FloatFromCurve(
     FloatCurve=DataInterface() {propertyValues {Curve {keys [
         {arriveTangent=-0.1447, interpMode=RCIM_Cubic, leaveTangent=-0.1447, tangentMode=RCTM_User, time=0, value=0}
         {arriveTangent=-0.0772, interpMode=RCIM_Cubic, leaveTangent=-0.0772, tangentMode=RCTM_User, time=0.0852, value=1.0091}
         <at 1 0>
        ]}}}
     Scale Curve=&FN_FrontEndFloatModifier(Gameplay=1, Front End=0.6)
    )
    Scale RGB=&VectorFromCurve(VectorCurve=DataInterface() {propertyValues {
       XCurve {keys [<at 0 0.0781>,<at 1 4.8262>,<at 1 0>]}
       YCurve {keys [<at 0 0.0095>,<at 1 5>,<at 1 0>]}
       ZCurve {keys [<at 0 0.056>,<at 1 5>,<at 1 0>]}
      }})
    ScaleA=true
    ScaleRGB=true
   )
   &CameraOffset(Camera Offset Amount=10)
  ]
  rendererClasses [&NiagaraSpriteRendererProperties]
  renderers [{rendererClass=&NiagaraSpriteRendererProperties}]
 }
 Lightning_CubeEdges {
  emitterUpdateScript [
   &EmitterState(Life Cycle Mode=::NewEnumerator1, Loop Behavior=::NewEnumerator1, Loop Duration=0.65, UseLoopDelay=true)
   &SpawnRate(
    SpawnRate=&FloatFromCurve(
     FloatCurve=DataInterface() {propertyValues {Curve {keys [{interpMode=RCIM_Cubic, tangentMode=RCTM_User, time=0, value=1}, {arriveTangent=-2.6311, tangentMode=RCTM_User, time=1, value=0}]}}}
     CurveIndex=$Emitter.NormalizedLoopAge
     Scale Curve=150
    )
    Use Spawn Probability=false
   ) {module=SpawnRate001}
  ]
  particleSpawnScript [
   &InitializeParticle(
    Lifetime=&RandomRangeFloat(Minimum=0.1, Maximum=0.25, Evaluation Type=::NewEnumerator1)
    Mesh Scale=/Niagara/DynamicInputs/CascadeConversion/SplitVector(
     Double Component Vector Channels:ECascadeNiagaraTwoVectorChannels=::NewEnumerator1
     Double Component:NiagaraFloat=&RandomRangeFloat(Minimum=0.5, Maximum=0.7, Evaluation Type=::NewEnumerator1)
     Single Component:NiagaraFloat=&RandomRangeFloat(Minimum=0.7, Maximum=0.7, Evaluation Type=::NewEnumerator1)
    )
    Mesh Scale Mode=::NewEnumerator4
    UsePositionOffset=false
    Write Material Random=true
   )
   &Color(Color=&MergeHelper_ColorSwitch(0=<rgba 13.5 1.56 15 0.5>, 1=<rgba 0 12.5 50 0.5>, 2=?, 3=?, Index=$System.VariantIndex, Num=::NewEnumerator0))
   &DynamicMaterialParameters(
    Index 0 Param 1=&RandomRangeFloat(Minimum=2, Maximum=3, Evaluation Type=::NewEnumerator1)
    Index 0 Param 3=&RandomRangeFloat(Minimum=0.2, Maximum=2, Evaluation Type=::NewEnumerator1)
    Index 0 Param 4=&RandomRangeFloat(Minimum=0, Maximum=1, Evaluation Type=::NewEnumerator1)
    Module.Param0WriteEnabled=true
    Module.Param1WriteEnabled=false
    Module.Param2WriteEnabled=true
    Module.Param3WriteEnabled=true
   )
   /SetVariables(Particles.Rnd:NiagaraInt32=&RandomRangeInt(Minimum=0, Maximum=7)) {isSetParameters=true}
   &InitialMeshOrientation(
    Mesh Orientation Mode=::NewEnumerator2
    Orientation Vector=&MergeHelper_VectorSwitch(
     0=&SelectVectorFromArray(
      Array Sampling Mode=::NewEnumerator0
      Direct Array Index=?
      Vector Selection Array=DataInterface() {propertyValues {FloatData [<-1 0 0>, <0 -1 0>, <0 0 1>]}}
      Random Seed=0
     )
     1=&SelectVectorFromArray(
      Array Sampling Mode=::NewEnumerator0
      Direct Array Index=?
      Vector Selection Array=DataInterface() {propertyValues {FloatData [<-1 0 0>, <0 1 0>, <0 0 1>]}}
      Random Seed=0
     )
     2=&SelectVectorFromArray(
      Array Sampling Mode=::NewEnumerator0
      Direct Array Index=?
      Vector Selection Array=DataInterface() {propertyValues {FloatData [<1 0 0>, <0 1 0>, <0 0 1>]}}
      Random Seed=0
     )
     3=&SelectVectorFromArray(
      Array Sampling Mode=::NewEnumerator0
      Direct Array Index=?
      Vector Selection Array=DataInterface() {propertyValues {FloatData [<1 0 0>, <0 -1 0>, <0 0 1>]}}
      Random Seed=0
     )
     4=&SelectVectorFromArray(
      Array Sampling Mode=::NewEnumerator0
      Direct Array Index=?
      Vector Selection Array=DataInterface() {propertyValues {FloatData [<-1 0 0>, <0 -1 0>, <0 0 -1>]}}
      Random Seed=0
     )
     5=&SelectVectorFromArray(
      Array Sampling Mode=::NewEnumerator0
      Direct Array Index=?
      Vector Selection Array=DataInterface() {propertyValues {FloatData [<-1 0 0>, <0 1 0>, <0 0 -1>]}}
      Random Seed=0
     )
     6=&SelectVectorFromArray(
      Array Sampling Mode=::NewEnumerator0
      Direct Array Index=?
      Vector Selection Array=DataInterface() {propertyValues {FloatData [<1 0 0>, <-0.25 1 0>, <-0.25 0 -1>]}}
      Random Seed=0
     )
     7=&SelectVectorFromArray(
      Array Sampling Mode=::NewEnumerator0
      Direct Array Index=?
      Vector Selection Array=DataInterface() {propertyValues {FloatData [<1 0 0.2>, <0 -1 0>, <0 0 -1>]}}
      Random Seed=0
     )
     Index=$Particles.Rnd
     Num=::NewEnumerator6
    )
    Rotation=&RandomRangeVector(Minimum=<0 0 0>, Maximum=<1 0 0>)
    Use Rotation Vector=true
   )
   /SetVariables(Particles.Position=&MergeHelper_VectorSwitch(
     0=/Niagara/DynamicInputs/Vectors/Position/AddVectorToPosition(Position=$Engine.Emitter.SimulationPosition, Vector=<9 0 -12>)
     1=&AddVectorToPosition(Position=$Engine.Emitter.SimulationPosition, Vector=<10 -20 -10>)
     2=&AddVectorToPosition(Position=$Engine.Emitter.SimulationPosition, Vector=<-10 -20 -10>)
     3=&AddVectorToPosition(Position=$Engine.Emitter.SimulationPosition, Vector=<-12 -2 -10>)
     4=&AddVectorToPosition(Position=$Engine.Emitter.SimulationPosition, Vector=<9 -1 11>)
     5=&AddVectorToPosition(Position=$Engine.Emitter.SimulationPosition, Vector=<9 -18 9>)
     6=&AddVectorToPosition(Position=$Engine.Emitter.SimulationPosition, Vector=<-5 -19 9>)
     7=&AddVectorToPosition(Position=$Engine.Emitter.SimulationPosition, Vector=<-10 0 10>)
     Index=$Particles.Rnd
     Num=::NewEnumerator6
    )) {isSetParameters=true}
  ]
  particleUpdateScript [
   &ParticleState()
   &ScaleMeshSize(Scale Factor=&VectorFromCurve(VectorCurve=DataInterface() {propertyValues {
       XCurve {keys [<at 0.0033 0.2 RCIM_Constant>,<at 0.1939 0.6 RCIM_Constant>,<at 0.3947 0.7606 RCIM_Constant>,<at 0.5948 0.8999 RCIM_Constant>,<at 0.8 1>]}
       YCurve {keys [<at 0.0033 1 RCIM_Constant>,<at 0.1939 0.8291 RCIM_Constant>,<at 0.3947 0.4373 RCIM_Constant>,<at 0.5948 0.3041 RCIM_Constant>,<at 0.8 0.2>]}
       ZCurve {keys [<at 0.0033 1 RCIM_Constant>,<at 0.1939 0.691 RCIM_Constant>,<at 0.3947 0.4457 RCIM_Constant>,<at 0.5948 0.5668 RCIM_Constant>,<at 0.8 0.6282>]}
      }}))
   &DynamicMaterialParameters(
    Index 0 Param 2=&FloatFromCurve(FloatCurve=DataInterface() {propertyValues {Curve {keys [{arriveTangent=-1.8894, interpMode=RCIM_Cubic, leaveTangent=-1.8894, tangentMode=RCTM_User, time=0, value=1}, <at 1 -0.25>]}}})
    Module.Param0WriteEnabled=false
    Module.Param1WriteEnabled=true
    Module.Param2WriteEnabled=false
    Module.Param3WriteEnabled=false
   ) {module=DynamicMaterialParameters001}
   &ScaleColor(
    Scale RGB=&VectorFromCurve(VectorCurve=DataInterface() {propertyValues {
       XCurve {keys [<at 0 50 RCIM_Cubic>, {arriveTangent=-50, interpMode=RCIM_Cubic, leaveTangent=-50, time=0.5, value=4.967}, <at 1 0>]}
       YCurve {keys [<at 0 19.1123 RCIM_Cubic>, {arriveTangent=-18.6172, interpMode=RCIM_Cubic, leaveTangent=-18.6172, time=0.5, value=2.739}, <at 1 0.4952>]}
       ZCurve {keys [<at 0 22.1016 RCIM_Cubic>, {arriveTangent=-17.1016, interpMode=RCIM_Cubic, leaveTangent=-17.1016, time=0.5, value=30}, <at 1 5>]}
      }})
    ScaleA=true
    ScaleRGB=true
   )
   &CameraOffset(Camera Offset Amount=1)
  ]
  rendererClasses [&NiagaraMeshRendererProperties]
  renderers [{rendererClass=&NiagaraMeshRendererProperties, rendererIndex=0}]
 }
 Mesh_Glow {
  emitterUpdateScript [
   &EmitterState(Life Cycle Mode=::NewEnumerator1, Loop Behavior=::NewEnumerator1, Loop Duration=1, UseLoopDelay=true)
   &SpawnBurst_Instantaneous(Use Spawn Probability=false, UseLoopCountLimit=false)
  ]
  particleSpawnScript [
   &InitializeParticle(
    Color=&MergeHelper_ColorSwitch(0=<rgba 1 0 0.85 0.5>, 1=<rgba 0.0067 0.863 1 0.5>, 2=?, 3=?, Index=$System.VariantIndex, Num=::NewEnumerator0)
    Lifetime=0.3
    Mesh Scale=<2.51 2.51 2.5>
    Mesh Scale Mode=::NewEnumerator4
    Sprite Size Mode=::NewEnumerator3
    Uniform Sprite Size=150
    UsePositionOffset=false
    Write Material Random=true
   )
   /SetVariables(
    Particles.FlareCamOffset=15
    Particles.FlareColor=&MergeHelper_ColorSwitch(0=<rgba 2.7 0.31 3 1>, 1=<rgba 0 2.2966 3 1>, 2=?, 3=?, Index=$System.VariantIndex, Num=::NewEnumerator0)
    Particles.FlareDyn01:Vector4f=<1 1 1 1>
    Particles.FlarePos:NiagaraPosition=&AddVectorToPosition(Position=$Engine.Emitter.SimulationPosition, Vector=<0 -10 0>)
   ) {isSetParameters=true}
  ]
  particleUpdateScript [
   &ParticleState()
   &CameraOffset(Camera Offset Amount=1)
   &DynamicMaterialParameters(
    Index 0 Param 1=&FloatFromCurve(FloatCurve=DataInterface() {propertyValues {Curve {keys [{interpMode=RCIM_Cubic, tangentMode=RCTM_User, time=0.25, value=0.05}, {arriveTangent=0.8598, tangentMode=RCTM_User, time=1, value=0.25}]}}})
    Index 0 Param 2=1
    Index 0 Param 3=1
    Index 0 Param 4=1
    Module.Param0WriteEnabled=true
    Module.Param1WriteEnabled=true
    Module.Param2WriteEnabled=true
    Module.Param3WriteEnabled=true
   )
   &ScaleColor(
    Scale RGB=&VectorFromFloat(Value=&FloatFromCurve(FloatCurve=DataInterface() {propertyValues {Curve {keys [
          {arriveTangent=0.4962, time=0.0045, value=0.0959}
          {arriveTangent=9.4097, interpMode=RCIM_Cubic, leaveTangent=9.4097, time=0.2968, value=0.5921}
          {arriveTangent=-0.017, interpMode=RCIM_Cubic, leaveTangent=-0.017, time=0.5096, value=4.8485}
          <at 1 0.5801>
         ]}}}))
    ScaleA=true
    ScaleRGB=true
   )
   /SetVariables(
    Particles.LightColor=&Multiply_LinearColorByFloat(
     Linear Color=&MergeHelper_ColorSwitch(0=<rgba 25 0 17 1>, 1=<rgba 0.0248 18.1971 25 1>, 2=?, 3=?, Index=$System.VariantIndex, Num=::NewEnumerator0)
     Float=&FloatFromCurve(FloatCurve=DataInterface() {propertyValues {Curve {keys [{interpMode=RCIM_Cubic, tangentMode=RCTM_User, time=0, value=1}, <at 1 0>]}}})
    )
    Particles.LightExponent=16
    Particles.LightPosition:NiagaraPosition=&AddVectorToPosition(Position=$Engine.Emitter.SimulationPosition, Vector=<0 -25 -10>)
    Particles.LightRadius=32
   ) {isSetParameters=true}
   /SetVariables(Particles.FlareColor=&Multiply_LinearColorByFloat(
     Linear Color=$Particles.Initial.FlareColor
     Float=&FloatFromCurve(
      FloatCurve=DataInterface() {propertyValues {Curve {keys [{interpMode=RCIM_Cubic, tangentMode=RCTM_User, time=0, value=1}, <at 0.1 0>]}}}
      CurveIndex=$Particles.Age
     )
    )) {isSetParameters=true}
  ]
  rendererClasses [
   &NiagaraMeshRendererProperties
   &NiagaraLightRendererProperties
   &NiagaraSpriteRendererProperties
  ]
  renderers [
   {rendererClass=&NiagaraMeshRendererProperties, rendererIndex=0}
   {rendererClass=&NiagaraLightRendererProperties, rendererIndex=1}
   {rendererClass=&NiagaraSpriteRendererProperties, rendererIndex=2}
  ]
 }
 Runes {
  emitterUpdateScript [
   &EmitterState(Life Cycle Mode=::NewEnumerator1, Loop Behavior=::NewEnumerator1, Loop Duration=0.125, UseLoopDelay=true)
   &SpawnRate(SpawnRate=50, Use Spawn Probability=false)
  ]
  particleSpawnScript [
   /SetVariables(Particles.SkeletonSamplingIndex:NiagaraInt32=/CascadeToNiagaraConverter/NiagaraScripts/DynamicInputScripts/SequentialInt(Use Array Data:NiagaraBool=false, First Number:NiagaraInt32=0, Sequence:NiagaraDataInterfaceArrayInt32=?, Last Number:NiagaraInt32=15)) {isSetParameters=true}
   &InitializeParticle(
    Lifetime=&RandomRangeFloat(Minimum=0.25, Maximum=0.5, Evaluation Type=::NewEnumerator1)
    Sprite Size=&Vector2DFromFloat(Value=&RandomRangeFloat(Minimum=4, Maximum=5, Evaluation Type=::NewEnumerator1))
    Sprite Size Mode=::NewEnumerator4
    UsePositionOffset=false
    Write Material Random=true
   )
   &Color(Color=&MergeHelper_ColorSwitch(
     0=&RandomRangeLinearColor(Minimum=<rgba 0.69 5 8 0.25>, Maximum=<rgba 0 15 4 0.5>)
     1=&RandomRangeLinearColor(Minimum=<rgba 0.69 5 8 0.25>, Maximum=<rgba 0 15 4 0.5>)
     2=?
     3=?
     Index=$System.VariantIndex
     Num=::NewEnumerator0
    ))
   /CascadeToNiagaraConverter/NiagaraScripts/ModuleScripts/AddVectorToPosition(Vector:Vector3f=&TransformVector(Vector=<0 -10 0>, Destination Space=::Simulation))
   &CameraOffset(Camera Offset Amount=&RandomRangeFloat(Minimum=2, Maximum=5, Evaluation Type=::NewEnumerator1))
   &ShapeLocation(MaskSpawnGroup=false, Override Seed=false, Sphere Radius=12, Sphere Surface Distribution=1, UseOwnerScale=true)
   /SetVariables(
    Particles.Electric_CamOffset=5
    Particles.Electric_Color:LinearColor=&MergeHelper_ColorSwitch(0=<rgba 50 0 38.25 0.25>, 1=<rgba 0.5679 32.0474 50 0.25>, 2=?, 3=?, Index=$System.VariantIndex, Num=::NewEnumerator0)
    Particles.Electric_Velocity:Vector3f=&RandomRangeVector(Minimum=<0.5 0 5>, Maximum=<1 0 2>, Evaluation Type=::NewEnumerator1)
   ) {isSetParameters=true}
   /SetVariables(Particles.RuneVis:NiagaraInt32=$System.VariantIndex) {isSetParameters=true}
  ]
  particleUpdateScript [
   &ParticleState()
   /SetVariables(
    Particles.Electric_Dyn01:Vector4f=&MakeVector4(
     X=&RandomRangeFloat(Minimum=2.5, Maximum=3.5, Evaluation Type=::NewEnumerator1)
     Y=&FloatFromCurve(FloatCurve=DataInterface() {propertyValues {Curve {keys [<at 0 1>,<at 1 -1>]}}})
     Z=&RandomRangeFloat(Minimum=0.2, Maximum=0.7, Evaluation Type=::NewEnumerator1)
     W=0
    )
    Particles.Electric_Rot:NiagaraFloat=&RandomRangeFloat(Minimum=0, Maximum=1)
    Particles.Electric_Size:Vector2f=&Vector2DFromFloat(Value=&FloatFromCurve(
      FloatCurve=DataInterface() {propertyValues {Curve {keys [{arriveTangent=1.6396, interpMode=RCIM_Cubic, leaveTangent=1.6396, tangentMode=RCTM_User, time=0.0079, value=0.4992}, <at 1 1>]}}}
      Scale Curve=10
     ))
   ) {isSetParameters=true}
   &ScaleSpriteSize(
    Scale Sprite Size Mode=::NewEnumerator0
    Uniform Scale Factor=&FloatFromCurve(FloatCurve=DataInterface() {propertyValues {Curve {keys [
         <at 0 1 RCIM_Cubic>
         {arriveTangent=1.6706, interpMode=RCIM_Cubic, leaveTangent=1.6706, time=0.043, value=2.109}
         {arriveTangent=-1.1588, interpMode=RCIM_Cubic, leaveTangent=-1.1588, time=0.2029, value=1.339}
         <at 1 1 RCIM_Cubic>
        ]}}})
   )
   &SubUVAnimation(
    End Frame Range Override=2
    "Sprite Renderer "=DataInterface() {propertyValues {SpriteRenderer=&self:Runes_0.NiagaraSpriteRendererProperties_2}}
    SubUV Animation Mode:ENiagara_SubUVLookupModeV2=::NewEnumerator2
    UseEndFrame=false
    UseStartFrame=false
   )
   &ScaleColor(
    Scale Alpha=&FloatFromCurve(FloatCurve=DataInterface() {propertyValues {Curve {keys [
         {arriveTangent=-0.1447, interpMode=RCIM_Cubic, leaveTangent=-0.1447, tangentMode=RCTM_User, time=0, value=0}
         {arriveTangent=-0.0772, interpMode=RCIM_Cubic, leaveTangent=-0.0772, tangentMode=RCTM_User, time=0.0852, value=1.0091}
         <at 1 0>
        ]}}})
    Scale RGB=&MergeHelper_VectorSwitch(
     0=&VectorFromCurve(VectorCurve=DataInterface() {propertyValues {XCurve {keys [<at 0 0>,<at 1 2>,<at 1 0>]}, YCurve {keys [<at 0 2>,<at 1 0>,<at 1 0>]}, ZCurve {keys [<at 0 0.25>,<at 1 0>,<at 1 0>]}}})
     1=&VectorFromCurve(VectorCurve=DataInterface() {propertyValues {XCurve {keys [<at 0 0>,<at 1 2>,<at 1 0>]}, YCurve {keys [<at 0 1>,<at 1 0>,<at 1 0>]}, ZCurve {keys [<at 0 2>,<at 1 0>,<at 1 0>]}}})
     2=?
     3=?
     Index=$System.VariantIndex
     Num=::NewEnumerator0
    )
    ScaleA=true
    ScaleRGB=true
   )
   &SkeletalMeshLocation(
    Bone / Socket Index=$Particles.SkeletonSamplingIndex
    Bone Sampling Mode=::NewEnumerator4
    MaskSpawnGroup=false
    Orientation Sampling=::NewEnumerator1
    Sampled Position Offset=&SelectVectorFromArray(
     Direct Array Index=$Particles.SkeletonSamplingIndex
     Vector Selection Array=DataInterface() {propertyValues {FloatData [
        <5 0 0>
        <5 0 0>
        <0 0 0>
        <-10 0 0>
        <5 0 0>
        <0 0 0>
        <-5 0 0>
        <10 0 0>
        <-10 0 0>
        <-8 0 -4>
        <-4 0 0>
        <8 0 4>
        <4 0 0>
        <0 0 0>
        <0 0 0>
        <0 0 0>
       ]}}
    )
    Skeletal Mesh=$System.Skeletal Mesh
   ) {enabled=false}
   &SolveForcesAndVelocity(Clamp Velocity=false, Limit Acceleration=false)
  ]
  rendererClasses [&NiagaraSpriteRendererProperties]
  renderers [
   {rendererClass=&NiagaraSpriteRendererProperties}
   {rendererClass=&NiagaraSpriteRendererProperties, rendererIndex=1}
   {rendererClass=&NiagaraSpriteRendererProperties, rendererIndex=2}
  ]
 }
 Swirls {
  emitterUpdateScript [
   &EmitterState(Life Cycle Mode=::NewEnumerator1, Loop Behavior=::NewEnumerator1, Loop Duration=1, UseLoopDelay=true)
   &SpawnBurst_Instantaneous(Use Spawn Probability=false, UseLoopCountLimit=false)
  ]
  particleSpawnScript [
   &InitialMeshOrientation(
    Mesh Axis To Orient=<0 0 1>
    Mesh Orientation Mode=::NewEnumerator2
    Orientation Vector=<0 0 1>
    Rotation=&RandomRangeVector(Minimum=<0 0 -360>, Maximum=<0 0 360>)
    Use Rotation Vector=true
   )
   &InitializeParticle(
    Lifetime=&RandomRangeFloat(Minimum=0.6, Maximum=1.2, Evaluation Type=::NewEnumerator1)
    Mesh Scale=&SplitVector(
     Double Component Vector Channels:ECascadeNiagaraTwoVectorChannels=::NewEnumerator0
     Double Component:NiagaraFloat=&RandomRangeFloat(Minimum=0.1, Maximum=0.25, Evaluation Type=::NewEnumerator1)
     Single Component:NiagaraFloat=&RandomRangeFloat(Minimum=0.5, Maximum=0.7, Evaluation Type=::NewEnumerator1)
    )
    Mesh Scale Mode=::NewEnumerator4
    UsePositionOffset=false
    Write Material Random=true
   )
   &Color(Color=<rgba 15 15 15 0.005>)
   &DynamicMaterialParameters(
    Index 0 Param 1=&RandomRangeFloat(Minimum=0, Maximum=0.25, Evaluation Type=::NewEnumerator1)
    Module.Param0WriteEnabled=true
    Module.Param1WriteEnabled=false
    Module.Param2WriteEnabled=false
    Module.Param3WriteEnabled=false
   )
  ]
  particleUpdateScript [
   &ParticleState()
   &ScaleMeshSize(Scale Factor=&VectorFromCurve(VectorCurve=DataInterface() {propertyValues {XCurve {keys [<at 0 1 RCIM_Cubic>,<at 1 1.1>]}, YCurve {keys [<at 0 1 RCIM_Cubic>,<at 1 1.1>]}, ZCurve {keys [<at 0 1 RCIM_Cubic>,<at 1 1>]}}}))
   &DynamicMaterialParameters(
    Index 0 Param 2=&FloatFromCurve(FloatCurve=DataInterface() {propertyValues {Curve {keys [{arriveTangent=0.7124, time=-0.0026, value=0.2876}, <at 1 1 RCIM_Cubic>]}}})
    Index 0 Param 3=1
    Index 0 Param 4=1
    Module.Param0WriteEnabled=false
    Module.Param1WriteEnabled=true
    Module.Param2WriteEnabled=true
    Module.Param3WriteEnabled=true
   ) {module=DynamicMaterialParameters001}
   &ScaleColor(
    Scale Alpha=&FloatFromCurve(FloatCurve=DataInterface() {propertyValues {Curve {keys [<at 0 1 RCIM_Cubic>,<at 1 0>]}}})
    Scale RGB=&MergeHelper_VectorSwitch(
     0=&VectorFromCurve(VectorCurve=DataInterface() {propertyValues {XCurve {keys [<at 0 1>,<at 1 0.7592>]}, YCurve {keys [<at 0 0.0384>,<at 1 0>]}, ZCurve {keys [<at 0 0.8441>,<at 1 1>]}}})
     1=&VectorFromCurve(VectorCurve=DataInterface() {propertyValues {XCurve {keys [<at 0 0>,<at 1 0>]}, YCurve {keys [<at 0 1>,<at 1 0.8>]}, ZCurve {keys [<at 0 0.8441>,<at 1 1>]}}})
     2=?
     3=?
     Index=$System.VariantIndex
     Num=::NewEnumerator0
    )
    ScaleA=true
    ScaleRGB=true
   )
  ]
  rendererClasses [&NiagaraMeshRendererProperties]
  renderers [{rendererClass=&NiagaraMeshRendererProperties, rendererIndex=0}]
 }
 Upward_Burst {
  emitterUpdateScript [
   &EmitterState(Life Cycle Mode=::NewEnumerator1, Loop Behavior=::NewEnumerator1, Loop Duration=1, UseLoopDelay=true)
   &SpawnBurst_Instantaneous(Spawn Group=1, Use Spawn Probability=false, UseLoopCountLimit=false) {module=SpawnBurst_Instantaneous001}
   &SpawnBurst_Instantaneous(Spawn Count=4, Use Spawn Probability=false, UseLoopCountLimit=false)
  ]
  particleSpawnScript [
   &InitializeParticle(
    Lifetime=&RandomRangeFloat(Minimum=0.25, Maximum=0.35, Evaluation Type=::NewEnumerator1)
    Sprite Size=&RandomRangeVector2D(Minimum=<10 15>, Maximum=<15 20>, Evaluation Type=::NewEnumerator1)
    Sprite Size Mode=::NewEnumerator4
    UsePositionOffset=false
    Write Material Random=true
   )
   &Color(Color=&MergeHelper_ColorSwitch(0=<rgba 0.97 0 1 0.35>, 1=<rgba 0.0151 0.7487 1 0.35>, 2=?, 3=?, Index=$System.VariantIndex, Num=::NewEnumerator0))
   &AddVelocity(Velocity=&RandomRangeVector(Minimum=<0 -5 25>, Maximum=<5 5 50>, Evaluation Type=::NewEnumerator1))
   &AddVectorToPosition(Vector:Vector3f=&TransformVector(Vector=<0 -10 0>, Destination Space=::Simulation))
   &AddVectorToPosition(Vector:Vector3f=&TransformVector(
     Vector=&RandomRangeVector(Minimum=<-5 -5 0>, Maximum=<5 5 0>, Evaluation Type=::NewEnumerator1)
     Destination Space=::Simulation
    )) {module=AddVectorToPosition001}
   &ShapeLocation(
    Hemisphere Distribution=<0.5 0.5>
    MaskSpawnGroup=false
    Override Seed=false
    Rotation Mode=::NewEnumerator2
    Sphere Radius=5
    Sphere Surface Distribution=1
    UseOwnerScale=true
    Yaw / Pitch / Roll=<0.75 0 0>
   )
   /SetVariables(
    Particles.RayColor:LinearColor=<rgba 15 15 15 0.25>
    Particles.RayPos:NiagaraPosition=/Niagara/DynamicInputs/Vectors/Position/AddVectorToPosition(Position=$Engine.Emitter.SimulationPosition, Vector=<0 -10 0>)
    Particles.RaySize:Vector2f=&RandomRangeVector2D(Minimum=<50 100>, Maximum=<100 125>, Evaluation Type=::NewEnumerator1)
    Particles.VisibilityTag=&SpawnGroup()
   ) {isSetParameters=true}
  ]
  particleUpdateScript [
   &ParticleState()
   &ScaleSpriteSize(
    Scale Factor=&Vector2DFromCurve(
     Vector2Curve=DataInterface() {propertyValues {
       XCurve {keys [{arriveTangent=0.5317, interpMode=RCIM_Cubic, leaveTangent=0.5317, tangentMode=RCTM_User, time=-0.0045, value=1}, <at 1 0.7>]}
       YCurve {keys [{arriveTangent=3.7386, interpMode=RCIM_Cubic, leaveTangent=3.7386, tangentMode=RCTM_User, time=-0.0045, value=0.0371}, <at 1 2>]}
      }}
     ScaleCurve=<1 1>
    )
    Scale Sprite Size Mode=::NewEnumerator2
   )
   &SolveForcesAndVelocity(Clamp Velocity=false, Limit Acceleration=false)
   &ScaleColor(
    Scale Alpha=&FloatFromCurve(
     FloatCurve=DataInterface() {propertyValues {Curve {keys [<at 0 1 RCIM_Cubic>,<at 0.75 0>]}}}
     Scale Curve=&FN_FrontEndFloatModifier(Gameplay=1, Front End=0.2)
    )
    Scale RGB=&VectorFromCurve(VectorCurve=DataInterface() {propertyValues {XCurve {keys [<at 0 50 RCIM_Cubic>,<at 1 5>]}, YCurve {keys [<at 0 50 RCIM_Cubic>,<at 1 5>]}, ZCurve {keys [<at 0 50 RCIM_Cubic>,<at 1 5>]}}})
    ScaleA=true
    ScaleRGB=true
   )
   &SubUVAnimation(
    End Frame Range Override=2
    "Sprite Renderer "=DataInterface() {propertyValues {SpriteRenderer=&self:Upward_Burst_0.NiagaraSpriteRendererProperties_5}}
    SubUV Animation Mode:ENiagara_SubUVLookupModeV2=::NewEnumerator2
    UseEndFrame=false
    UseStartFrame=false
   )
   &SubUVAnimation(
    End Frame Range Override=2
    "Sprite Renderer "=DataInterface() {propertyValues {SpriteRenderer=&self:Upward_Burst_0.NiagaraSpriteRendererProperties_10}}
    UseEndFrame=false
    UseStartFrame=false
   ) {module=SubUVAnimation001}
   /SetVariables(
    Particles.RayCamOffset=25
    Particles.RayColor:LinearColor=&MakeLinearColorFromVectorAndFloat(
     "Vector (RGB)"=&MergeHelper_VectorSwitch(
      0=&VectorFromCurve(
       VectorCurve=DataInterface() {propertyValues {
         XCurve {keys [
           {arriveTangent=-0.139, time=0.25, value=0.3097}
           {arriveTangent=-0.1114, interpMode=RCIM_Cubic, leaveTangent=-0.1114, time=0.6, value=0.2777}
           {arriveTangent=-1.1392, leaveTangent=-1.1392, time=0.6738, value=0.2624}
           <at 1 0>
          ]}
         YCurve {keys [
           {arriveTangent=0.0234, time=0.25, value=0.0523}
           {arriveTangent=0.1447, interpMode=RCIM_Cubic, leaveTangent=0.1447, time=0.6, value=0.0868}
           {arriveTangent=-0.4345, leaveTangent=-0.4345, time=0.6738, value=0.1136}
           <at 1 0>
          ]}
         ZCurve {keys [
           {arriveTangent=0.0476, time=0.25, value=0.4167}
           {arriveTangent=1.3765, interpMode=RCIM_Cubic, leaveTangent=1.3765, time=0.6, value=0.3906}
           {arriveTangent=-2.0374, leaveTangent=-2.0374, time=0.6738, value=1}
           <at 1 0>
          ]}
        }}
       Scale Curve=<15 15 15>
      )
      1=&VectorFromCurve(
       VectorCurve=DataInterface() {propertyValues {
         XCurve {keys [
           {arriveTangent=-0.139, time=0.25, value=0.3097}
           {arriveTangent=-0.1114, interpMode=RCIM_Cubic, leaveTangent=-0.1114, time=0.6, value=0.2777}
           {arriveTangent=-1.1392, leaveTangent=-1.1392, time=0.6738, value=0.2624}
           <at 1 0>
          ]}
         YCurve {keys [
           {arriveTangent=0.0234, time=0.25, value=0.0523}
           {arriveTangent=0.1447, interpMode=RCIM_Cubic, leaveTangent=0.1447, time=0.6, value=0.0868}
           {arriveTangent=-0.4345, leaveTangent=-0.4345, time=0.6738, value=0.1136}
           <at 1 0>
          ]}
         ZCurve {keys [
           {arriveTangent=0.0476, time=0.25, value=0.4167}
           {arriveTangent=1.3765, interpMode=RCIM_Cubic, leaveTangent=1.3765, time=0.6, value=0.3906}
           {arriveTangent=-2.0374, leaveTangent=-2.0374, time=0.6738, value=1}
           <at 1 0>
          ]}
        }}
       Scale Curve=<0 15 15>
      )
      2=?
      3=?
      Index=$System.VariantIndex
      Num=::NewEnumerator0
     )
     "Float (Alpha)"=&FloatFromCurve(
      FloatCurve=DataInterface() {propertyValues {Curve {keys [
          {arriveTangent=-0.1447, interpMode=RCIM_Cubic, leaveTangent=-0.1447, tangentMode=RCTM_User, time=0, value=0}
          {arriveTangent=-0.0772, interpMode=RCIM_Cubic, leaveTangent=-0.0772, tangentMode=RCTM_User, time=0.0186, value=1.0006}
          <at 1 0>
         ]}}}
      Scale Curve=&FN_FrontEndFloatModifier(Gameplay=0.25, Front End=0.02)
     )
    )
    Particles.RaySize:Vector2f=&Multiply_Vector2D(
     A=$Particles.Initial.RaySize
     B=&Vector2DFromCurve(
      Vector2Curve=DataInterface() {propertyValues {
        XCurve {keys [
          {arriveTangent=2.4707, interpMode=RCIM_Cubic, leaveTangent=2.4707, tangentMode=RCTM_User, time=0, value=1.4793}
          {arriveTangent=0.5207, interpMode=RCIM_Cubic, leaveTangent=0.5207, time=0.6461, value=1.8188}
          <at 1 2 RCIM_Cubic>
         ]}
        YCurve {keys [
          {arriveTangent=1.1192, interpMode=RCIM_Cubic, leaveTangent=1.1192, tangentMode=RCTM_User, time=0, value=0.5}
          {arriveTangent=0.5, interpMode=RCIM_Cubic, leaveTangent=0.5, time=0.6461, value=0.9321}
          <at 1 1 RCIM_Cubic>
         ]}
       }}
      ScaleCurve=<1 1>
     )
    )
   ) {isSetParameters=true}
  ]
  rendererClasses [&NiagaraSpriteRendererProperties]
  renderers [
   {rendererClass=&NiagaraSpriteRendererProperties}
   {rendererClass=&NiagaraSpriteRendererProperties, rendererIndex=1}
  ]
 }
}
summary {
 systemName=NS_BackPack_EvilCowboy_Stage3_Transition
 emitters [
  {emitterName=Lightning_CubeEdges, rendererClasses [&NiagaraMeshRendererProperties]}
  {emitterName=Cube_WispAway, rendererClasses [&NiagaraMeshRendererProperties]}
  {emitterName=Runes, rendererClasses [&NiagaraSpriteRendererProperties]}
  {emitterName=LightRays, rendererClasses [&NiagaraSpriteRendererProperties]}
  {emitterName=Upward_Burst, rendererClasses [&NiagaraSpriteRendererProperties]}
  {emitterName=InkySmokeRadial02, rendererClasses [&NiagaraSpriteRendererProperties]}
  {emitterName=Swirls, rendererClasses [&NiagaraMeshRendererProperties]}
  {
   emitterName=Mesh_Glow
   rendererClasses [
    &NiagaraMeshRendererProperties
    &NiagaraLightRendererProperties
    &NiagaraSpriteRendererProperties
   ]
  }
 ]
 userVariables [
  User.bIsFrontEnd {type=NiagaraBool}
  User.bIsFrontEndPreview {type=NiagaraBool}
  User.Variant {type=NiagaraFloat, defaultValue=0, description=""}
 ]
}
