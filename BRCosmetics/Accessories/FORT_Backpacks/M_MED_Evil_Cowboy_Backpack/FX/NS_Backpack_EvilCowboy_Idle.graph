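# Niagara idle effect for the Evil Cowboy backpack cosmetic that emits a loop of small flame sprites around the pack.
# One emitter spawns flipbook flames in an offset box, stretches and fades them over life, and pushes them toward the camera.
# A system variant index switches the color between green and blue, with a separate green range for front-end preview.
# The visibility user bool gates spawning by zeroing spawn probability.
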NiagaraSystem /BRCosmetics/Accessories/FORT_Backpacks/M_MED_Evil_Cowboy_Backpack/FX/NS_Backpack_EvilCowboy_Idle
dependencies {
 usedDataInterfaces [
  &NiagaraDataInterfaceCurve
  &NiagaraDataInterfaceVector2DCurve
  &NiagaraDataInterfaceSpriteRendererInfo
 ]
 usedDynamicInputs [&MakeIntFromFloat]
 usedModules [
  &SystemState
  &self:SystemSpawnScript.SystemScriptSource.SystemScriptGraph.NiagaraNodeAssignment_0.SetVariables_#1
 ]
}
emitters {Flames_Random {
  emitterUpdateScript [
   &EmitterState()
   &SpawnRate(
    Spawn Probability=&MakeCustomFloatFromBool(
     Bool=$User.bIsVisible
     True Float=&MergeHelper_FloatSwitch(0=1, 1=1, 2=?, 3=?, Index=$System.VariantIndex, Num=::NewEnumerator0)
     False Float=0
    )
    SpawnRate=15
    Use Spawn Probability=true
   )
  ]
  particleSpawnScript [
   &InitializeParticle(
    Color=&MergeHelper_ColorSwitch(
     0=&MakeCustomLinearColorFromBool(
      Bool=$User.bIsFrontEndPreview
      True Color=&RandomRangeLinearColor(Minimum=<rgba 0.1773 0.5 0.1395 0.1>, Maximum=<rgba 1.9243 5 1.2835 0.1>)
      False Color=&RandomRangeLinearColor(Minimum=<rgba 0.1272 1 0.0425 0.2>, Maximum=<rgba 1.1477 5 0.2091 0.225>)
     )
     1=&RandomRangeLinearColor(Minimum=<rgba 0 0.33 1 0.35>, Maximum=<rgba 0 2.5 5 0.3>)
     2=?
     3=?
     Index=$System.VariantIndex
     Num=::NewEnumerator0
    )
    Sprite Size Mode=::NewEnumerator1
    Uniform Sprite Size Max=15
    Uniform Sprite Size Min=10
    UsePositionOffset=false
    Write Material Random=true
   )
   &ShapeLocation(
    Box Size=<8 13 18>
    MaskSpawnGroup=false
    Offset=<0 -10 0>
    Offset Mode=::NewEnumerator0
    Override Seed=false
    Shape Primitive=::NewEnumerator2
    Surface Only=false
    UseOwnerScale=true
   )
  ]
  particleUpdateScript [
   &ParticleState()
   &ScaleSpriteSize(
    Non-Uniform Curve Scale=&RandomRangeVector2D(Minimum=<0.25 0.75>, Maximum=<0.75 1>)
    Non-Uniform Curve Sprite Scale=DataInterface() {propertyValues {
      XCurve {keys [
        {interpMode=RCIM_Cubic, tangentMode=RCTM_User, time=0, value=0}
        {interpMode=RCIM_Cubic, tangentMode=RCTM_User, time=0.15, value=1.5}
        {interpMode=RCIM_Cubic, tangentMode=RCTM_User, time=0.5, value=1}
        {interpMode=RCIM_Cubic, tangentMode=RCTM_User, time=1, value=1}
       ]}
      YCurve {keys [
        {arriveTangent=2.8715, interpMode=RCIM_Cubic, leaveTangent=2.8715, tangentMode=RCTM_User, time=0.1, value=0.5}
        {arriveTangent=0.0901, interpMode=RCIM_Cubic, leaveTangent=0.0901, tangentMode=RCTM_User, time=0.9848, value=1.5}
       ]}
     }}
    Scale Sprite Size Mode=::NewEnumerator3
   )
   &SubUVAnimation(
    End Frame Range Override=62
    "Sprite Renderer "=DataInterface() {propertyValues {SpriteRenderer=&self:Flames_Random_0.NiagaraSpriteRendererProperties_1}}
    Start Frame Range Override=16
    UseEndFrame=true
    UseStartFrame=true
   )
   &ScaleColor(
    Scale Alpha=&FloatFromCurve(FloatCurve=DataInterface() {propertyValues {Curve {keys [
         {arriveTangent=11.7282, interpMode=RCIM_Cubic, leaveTangent=11.7282, tangentMode=RCTM_User, time=0, value=0}
         <at 0.2 1 RCIM_Cubic>
         <at 1 0>
        ]}}})
    ScaleA=true
    ScaleRGB=true
   )
   &CameraOffset(Camera Offset Amount=20)
  ]
  rendererClasses [&NiagaraSpriteRendererProperties]
  renderers [{rendererClass=&NiagaraSpriteRendererProperties}]
 }}
summary {
 systemName=NS_Backpack_EvilCowboy_Idle
 emitters [{emitterName=Flames_Random, rendererClasses [&NiagaraSpriteRendererProperties]}]
 userVariables [
  User.bIsFrontEnd {type=NiagaraBool}
  User.bIsFrontEndPreview {type=NiagaraBool}
  User.bIsVisible {type=NiagaraBool}
  User.Variant {type=NiagaraFloat, defaultValue=0, description=""}
 ]
}
