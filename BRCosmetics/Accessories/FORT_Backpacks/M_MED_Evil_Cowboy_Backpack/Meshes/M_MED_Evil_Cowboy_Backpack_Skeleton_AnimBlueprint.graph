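# Post-process style AnimBlueprint for the Evil Cowboy backpack cosmetic, adding physics-driven secondary motion to an input pose.
# Four Anim Dynamics nodes run in series on the main pack, top-left and top-right chains, and bottom-left chain bones.
# The pack uses a stiff angular spring, the top chains linear springs, and the bottom chain only damped, tightly limited motion.
# All use gravity overrides and scaled component movement; the event graph has no logic.
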AnimBlueprint /BRCosmetics/Accessories/FORT_Backpacks/M_MED_Evil_Cowboy_Backpack/Meshes/M_MED_Evil_Cowboy_Backpack_Skeleton_AnimBlueprint
target_skeleton=&/M_MED_Evil_Cowboy_Backpack_Skeleton
graphs [
 AnimGraph {nodes [
   AnimGraphNode_SubInput_0 = Input Pose()
   n1 = Anim Dynamics - Bone: dyn_mainPack(ComponentPose=Local To Component(LocalPose=@AnimGraphNode_SubInput_0), Alpha=1, ExternalForce="0,0,0") {
    angularDampingOverride=0.985
    angularSpringConstant=10000
    bAngularSpring=true
    bGravityOverrideInSimSpace=true
    bOverrideAngularDamping=true
    bUseGravityOverride=true
    alphaBoolBlend {blendOption=Linear}
    boundBone {boneName=dyn_mainPack}
    componentAppliedLinearAccClamp=<1000 1000 1000>
    componentLinearAccScale=<0.3 3 0.3>
    componentLinearVelScale=<0.6 0.6 0.6>
    gravityOverride {z=-980}
    physicsBodyDefinitions [{
      sphereCollisionRadius=0
      boundBone {boneName=dyn_mainPack}
      boxExtents=<20 20 20>
      constraintSetup {angularLimitsMax=<10 0 0>, angularLimitsMin=<-10 0 0>, angularTarget=<0 0 0>}
      localJointOffset=<-0.5 0 -0.01>
     }]
    relativeSpaceBone {boneName=backpack_root}
   }
   n2 = Anim Dynamics - Bone: dyn_chainTop_l(ComponentPose=@n1, Alpha=1, ExternalForce="0,0,0") {
    angularSpringConstant=300
    bGravityOverrideInSimSpace=true
    bLinearSpring=true
    bOverrideLinearDamping=true
    bUseGravityOverride=true
    linearDampingOverride=0.9999
    linearSpringConstant=89
    alphaBoolBlend {blendOption=Linear}
    boundBone {boneName=dyn_chainTop_l}
    componentAppliedLinearAccClamp=<1000 1000 1000>
    componentLinearAccScale=<0.5 0.5 0.5>
    componentLinearVelScale=<0.5 0.5 0.5>
    gravityOverride {z=-980}
    physicsBodyDefinitions [{
      sphereCollisionRadius=0
      boundBone {boneName=dyn_chainTop_l}
      boxExtents=<1.5 2 2>
      constraintSetup {angularLimitsMax=<0 5 0>, angularLimitsMin=<0 -5 0>, angularTarget=<0 0 0>, linearAxesMax=<0 0 1>, linearAxesMin=<0 0 -1>}
      localJointOffset=<-0.125 0 0>
     }]
   }
   n3 = Anim Dynamics - Bone: dyn_chainTop_r(ComponentPose=@n2, Alpha=1, ExternalForce="0,0,0") {
    angularSpringConstant=300
    bGravityOverrideInSimSpace=true
    bLinearSpring=true
    bOverrideLinearDamping=true
    bUseGravityOverride=true
    linearDampingOverride=0.9999
    linearSpringConstant=75
    alphaBoolBlend {blendOption=Linear}
    boundBone {boneName=dyn_chainTop_r}
    componentAppliedLinearAccClamp=<1000 1000 1000>
    componentLinearAccScale=<0.5 0.5 0.5>
    componentLinearVelScale=<0.5 0.5 0.5>
    gravityOverride {z=-980}
    physicsBodyDefinitions [{
      sphereCollisionRadius=0
      boundBone {boneName=dyn_chainTop_r}
      boxExtents=<15 12 11>
      constraintSetup {angularLimitsMax=<0 5 0>, angularLimitsMin=<0 -5 0>, angularTarget=<0 0 0>, linearAxesMax=<0 0 1>, linearAxesMin=<0 0 -1>}
      localJointOffset=<-0.2 0 0>
     }]
   }
   n4 = Anim Dynamics - Bone: dyn_chainBottom_l(ComponentPose=@n3, Alpha=1, ExternalForce="0,0,0") {
    angularDampingOverride=0.9985
    angularSpringConstant=250
    bGravityOverrideInSimSpace=true
    bOverrideAngularDamping=true
    bOverrideLinearDamping=true
    bUseGravityOverride=true
    linearDampingOverride=0.998
    linearSpringConstant=250
    alphaBoolBlend {blendOption=Linear}
    boundBone {boneName=dyn_chainBottom_l}
    componentAppliedLinearAccClamp=<1000 1000 1000>
    componentLinearAccScale=<0.35 0.35 0.35>
    componentLinearVelScale=<0.45 0.45 0.45>
    gravityOverride {z=-980}
    physicsBodyDefinitions [{
      sphereCollisionRadius=0
      boundBone {boneName=dyn_chainBottom_l}
      boxExtents=<4 8 8>
      constraintSetup {angularLimitsMax=<0 0 8>, angularLimitsMin=<0 0 -8>, angularTarget=<0 0 0>, linearAxesMax=<0 0.1 0>, linearAxesMin=<0 -0.1 0>}
      localJointOffset=<-1 -0.1 0>
     }]
   }
   Output Pose(Result=Component To Local(ComponentPose=@n4)) {layerGroup=None}
  ]}
 EventGraph {nodes [Event BlueprintUpdateAnimation()]}
]
variables []
